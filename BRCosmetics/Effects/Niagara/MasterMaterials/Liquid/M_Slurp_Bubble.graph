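# Masked, two-sided material for Niagara mesh particles that renders a slurp liquid bubble which can erode.
# Instance-space, world-aligned cell noise and Fresnel shape the emissive and erosion mask, with two-sided sign separating front and back faces.
# A mask from particle-center distance and normal Z blends cell-noise normals into MF_Slurp_forFX and drives displacement along normals.
# Dynamic parameters control erosion, noise tiling, displacement strength and mask radius.
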Material /BRCosmetics/Effects/Niagara/MasterMaterials/Liquid/M_Slurp_Bubble
graph {
 properties {
  bUsedWithNiagaraMeshParticles=true
  blendMode=BLEND_Masked
  opacityMaskClipValue=0.01
  pixelDepthOffsetMode=PDOM_Legacy
  translucencyLightingMode=TLM_Surface
  twoSided=true
  usedShadingModels=Default Lit
 }
 nodes [
  ActorPositionWS()
  Constant() {r=1}
  MaterialFunctionCall() {materialFunction=&DebugTimeSine}
  n1 = NamedRerouteUsage()
  n2 = Fresnel() {baseReflectFraction=0, exponent=1}
  n3 = MaterialFunctionCall(In=@n2, Contrast=Multiply(A=NamedRerouteUsage(), B=Constant() {r=0.25})) {materialFunction=&CheapContrast}
  n4 = DynamicParameter() {defaultValue=<rgba 0.4 1 0 1>, paramNames [Erode 1-0, NoiseTIle, WPO, Param4]}
  n5 = Multiply(A=@n4.NoiseTIle, B=ScalarParameter() {defaultValue=60, parameterName=ErodeTile})
  n6 = TransformPosition(None=WorldPosition() {worldPositionShaderOffset=WPT_ExcludeAllShaderOffsets}.XYZ) {transformSourceType=TRANSFORMPOSSOURCE_World, transformType=TRANSFORMPOSSOURCE_Instance}
  n7 = MaterialFunctionCall(
   TextureObject=TextureObject() {samplerType=SAMPLERTYPE_LinearGrayscale, texture=/Game/Effects/Fort_Effects/Textures/Tile/T_CellNoise_Tile_M}
   TextureSize=@n5
   WorldPosition=@n6
  ) {materialFunction=/Game/Weapons/WeaponSkins/Wraps/MaterialFunctions/WorldAlignedTexture}
  n8 = OneMinus(None=@n2)
  n9 = Add(
   A=NamedRerouteDeclaration(None=ComponentMask(None=@n7.Z Texture) {g=true}) {name=Texture, nodeColor=<rgba 0 1 0.9297 1>}
   B=Multiply(A=Power(Base=@n8) {constExponent=8}) {constB=0.5}
  )
  n10 = LinearInterpolate(A=@n9, Alpha=@n3.Result) {constB=0}
  n11 = LinearInterpolate(
   A=Multiply(A=ConstantBiasScale(None=@n1) {bias=0.5, scale=2}, B=OneMinus(None=@n3.Result))
   B=Multiply(A=Power(Base=@n10), B=NamedRerouteUsage())
   Alpha=ConstantBiasScale(None=TwoSidedSign())
  )
  n12 = Saturate(None=Multiply(A=NamedRerouteUsage()) {constB=2})
  n13 = VertexNormalWS()
  n14 = WorldPosition()
  n15 = ParticlePositionWS()
  n16 = ScalarParameter() {defaultValue=15, parameterName=MeshRadius}
  n17 = LinearInterpolate(Alpha=Divide(A=Distance(A=@n14.XYZ, B=@n15), B=Multiply(A=@n16, B=@n4.Param4))) {constA=-1}
  n18 = DotProduct(
   A=Transform(None=VertexNormalWS()) {transformSourceType=TRANSFORMSOURCE_World, transformType=TRANSFORM_Instance}
   B=Constant3Vector() {constant=<rgba 0 0 1 0>}
  )
  n19 = Power(Base=Saturate(None=LinearInterpolate(A=@n17, B=@n18) {constAlpha=0.3}))
  n20 = Multiply(A=DotProduct(A=Abs(None=@n13), B=Constant3Vector() {constant=<rgba 0 0 1 0>}), B=@n19)
  n21 = LinearInterpolate(
   A=Constant3Vector() {constant=<rgba 0 0 1 0>}
   B=MaterialFunctionCall(
    TextureObject=TextureObject() {samplerType=SAMPLERTYPE_Normal, texture=/Game/Effects/Fort_Effects/Textures/Tile/T_CellNoise_Tile_N}
    TextureSize=@n5
    WorldPosition=@n6
   ) {materialFunction=&WorldAlignedNormal}.Z Texture
   Alpha=Saturate(None=Multiply(A=@n12, B=@n20))
  ) {constAlpha=1}
  n22 = GetMaterialAttributes(None=MaterialFunctionCall!(Pre Skinned Position, World Normal=Transform(None=@n21), AO) {materialFunction=/Game/Effects/Niagara/MasterMaterials/Liquid/MF_Slurp_forFX}.Result) {attributeGetTypes [
    #1
    #2
    #3
    #4
    #5
   ]}
  MaterialFunctionCall(Emissive=Multiply(A=Saturate(None=@n11), B=@n22.Emissive Color)) {materialFunction=&MF_FX_EmissiveAdjustments}
  MaterialFunctionCall() {materialFunction=&ObjectScale}
  Multiply(A=@n22.Base Color, B=LinearInterpolate(Alpha=NamedRerouteUsage()) {constA=0.6})
  Multiply(A=@n22.Roughness, B=LinearInterpolate(Alpha=NamedRerouteUsage()) {constA=1, constB=0.8})
  n23 = LinearInterpolate(Alpha=@n19) {constA=-1}
  Multiply(A=@n4.WPO, B=Multiply(A=Multiply(B=@n23) {constA=1}, B=VertexNormalWS()))
  NamedRerouteDeclaration(None=@n4.Erode 1-0) {name=ErodeParam, nodeColor=<rgba 0 0.5312 1 1>}
  NamedRerouteDeclaration(None=@n19) {name=VerticalMask, nodeColor=<rgba 1 0 0.7734 1>}
  Reroute(None=@n21)
  n24 = NamedRerouteUsage()
  n25 = LinearInterpolate(B=@n24, Alpha=Multiply(A=NamedRerouteUsage(), B=@n9)) {constA=1}
  n26 = LinearInterpolate(A=ScalarParameter() {defaultValue=50, parameterName=EndPower}, B=ScalarParameter() {defaultValue=15, parameterName=StartPower}, Alpha=@n24) {constA=50, constB=-1}
  Reroute(None=Multiply(A=NamedRerouteUsage(), B=Power(Base=@n25, Exp=@n26)))
  ScalarParameter() {defaultValue=0.612}
  Transform(None=Constant3Vector() {constant=<rgba 1 1 0 0>}) {transformSourceType=TRANSFORMSOURCE_Local, transformType=TRANSFORM_Instance}
 ]
}
